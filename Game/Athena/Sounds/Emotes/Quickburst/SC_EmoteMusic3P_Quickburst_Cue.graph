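# Sound cue for the third-person music of the Quickburst emote.
# A mixer starts the intro wave immediately alongside a looping wave held back by a fixed delay of under a second, wrapped in a single-entry Random node.
# It uses synced wide emote attenuation, emote music class and concurrency, keeps playing when silent, and sends to a music audio-analysis submix.
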SoundCue /Game/Athena/Sounds/Emotes/Quickburst/SC_EmoteMusic3P_Quickburst_Cue
root=Mixer_0
nodes [
 {id=Delay_0, children [Random_0], properties {delayMax=0.8, delayMin=0.8}}
 {id=Mixer_0, children [WavePlayer_0,Delay_0], properties {inputVolume [1,1]}}
 {id=Random_0, children [WavePlayer_1], properties {weights [1]}}
 {id=WavePlayer_0, properties {soundWaveAssetPtr=&/Emote_Quickburst_Music_Intro}}
 {id=WavePlayer_1, properties {bLooping=true, soundWaveAssetPtr=&/Emote_Quickburst_Music_Loop}}
]
properties {
 attenuationSettings=/Game/Sounds/Attenuation/Fort_Emotes_Athena_Sync_Wide
 maxDistance=3500
 soundClassObject=/Game/Sounds/FortSoundClassesAndModes/Fort_Music_Emote_3P
 virtualizationMode=PlayWhenSilent
 volumeMultiplier=0.7
 concurrencySet [/Game/Sounds/Concurrency_Rules/Athena_EmoteMusic_Conc]
 soundSubmixSends [{minSendLevel=0, soundSubmix=&Fortnite_AudioAnalysis_Music}]
}
