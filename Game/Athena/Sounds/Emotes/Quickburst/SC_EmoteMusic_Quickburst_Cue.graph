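# Sound cue for the Quickburst emote music in Athena, using the first-person emote music sound class and concurrency.
# A licensed-content switcher picks between looping silence and the music mix, which fills its other two inputs.
# The mix layers a one-shot intro with a looping track delayed by a fixed 0.8 seconds through a single-entry random node.
# Output routes to the licensed-content 3D submix with a send to music audio analysis, and keeps playing when silent.
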SoundCue /Game/Athena/Sounds/Emotes/Quickburst/SC_EmoteMusic_Quickburst_Cue
root=FortSoundNodeLicensedContentSwitcher_0
nodes [
 {id=Delay_0, children [Random_0], properties {delayMax=0.8, delayMin=0.8}}
 {id=FortSoundNodeLicensedContentSwitcher_0, class=FortSoundNodeLicensedContentSwitcher, children [WavePlayer_0,Mixer_0,Mixer_0]}
 {id=Mixer_0, children [WavePlayer_1,Delay_0], properties {inputVolume [1,1]}}
 {id=Random_0, children [WavePlayer_2], properties {weights [1]}}
 {id=WavePlayer_0, properties {bLooping=true, soundWaveAssetPtr=/Game/Sounds/Fort_Test_Sounds/5_seconds_silence}}
 {id=WavePlayer_1, properties {soundWaveAssetPtr=&/Emote_Quickburst_Music_Intro}}
 {id=WavePlayer_2, properties {bLooping=true, soundWaveAssetPtr=&/Emote_Quickburst_Music_Loop}}
]
properties {
 maxDistance=2097152
 soundClassObject=/Game/Sounds/FortSoundClassesAndModes/Fort_Music_Emote_1P_PSM
 soundSubmixObject=/Game/Sounds/SubMixes/Fortnite_LicensedContent_3D
 virtualizationMode=PlayWhenSilent
 volumeMultiplier=0.7
 concurrencySet [/Game/Sounds/Concurrency_Rules/Athena_EmoteMusic_1P_Conc]
 soundSubmixSends [{minSendLevel=0, soundSubmix=&Fortnite_AudioAnalysis_Music}]
}
